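# Post-process style AnimBlueprint for the MustardToast male body cosmetic, adding secondary motion to vest straps over the pose copied from the attached parent mesh.
# AnimDynamics simulates both dynamic straps, then a Control Rig and state-driven additive bone offsets correct them.
# Cached poses give separate strap offsets for skydiving and other states.
# The update event computes gravity override, a brief crouch-sprint correction pulse and clavicle pitch/roll values.
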AnimBlueprint /BRCosmetics/Characters/Player/Male/Medium/Bodies/M_MED_MustardToast/Meshes/M_MED_MustardToast_AnimBP
target_skeleton=&/M_MED_MustardToast_Skeleton
graphs [
 AnimGraph {nodes [
   Output Pose(Result=Blend Poses by bool(BlendPose_0=Use cached pose 'skydiving'(), BlendPose_1=Use cached pose 'common cache'(), bActiveValue=Get bIsSkyDiving()) {blendType=Linear}) {layerGroup=None}
   n1 = Get Gravity Override()
   n2 = Anim Dynamics - Bone: dyn_vest_strap_l!(
    ComponentPose=Local To Component(LocalPose=Copy Pose From Mesh() {bUseAttachedParent=true})
    Alpha=1
    GravityOverride:Vector=@n1
    ExternalForce="0,0,0"
   ) -> (Pose) {
    angularDampingOverride=0.9999
    bOverrideAngularDamping=true
    bUseGravityOverride=true
    simulationSpace=RootRelative
    alphaBoolBlend {blendOption=Linear}
    boundBone {boneName=dyn_vest_strap_l}
    componentAppliedLinearAccClamp=<200 800 500>
    componentLinearAccScale=<0.2 0.6 0.3>
    componentLinearVelScale=<0.5 0.7 0.4>
    physicsBodyDefinitions [{
      collisionType=CustomSphere
      sphereCollisionRadius=1
      boundBone {boneName=dyn_vest_strap_l}
      boxExtents=<17 17 17>
      constraintSetup {angularTargetAxis=AxisY, twistAxis=AxisZ, angularLimitsMax=<25 50 0>, angularLimitsMin=<-20 -50 0>, angularTarget=<0 0 0>}
      localJointOffset=<0 0 -2>
     }]
    planarLimits [{drivingBone {boneName=pelvis}, planeTransform {location=<0.654 7.4029 -10.5709>, rotation=<rot 3.8635 -6.0721 122.3506>}}]
   }
   n3 = Anim Dynamics - Bone: dyn_vest_strap_r!(ComponentPose=@n2, Alpha=1, GravityOverride:Vector=@n1, ExternalForce="0,0,0") -> (Pose) {
    angularDampingOverride=0.9999
    angularSpringConstant=300
    bOverrideAngularDamping=true
    bUseGravityOverride=true
    simulationSpace=RootRelative
    alphaBoolBlend {blendOption=Linear}
    boundBone {boneName=dyn_vest_strap_r}
    componentAppliedLinearAccClamp=<200 800 500>
    componentLinearAccScale=<0.2 0.3 0.3>
    componentLinearVelScale=<0.5 0.6 0.4>
    physicsBodyDefinitions [{
      collisionType=CustomSphere
      sphereCollisionRadius=1
      boundBone {boneName=dyn_vest_strap_r}
      boxExtents=<17 17 17>
      constraintSetup {angularTargetAxis=AxisY, twistAxis=AxisZ, angularLimitsMax=<20 50 0>, angularLimitsMin=<-20 -35 0>, angularTarget=<0 0 0>}
      localJointOffset=<0 0 -3.5>
     }]
    planarLimits [{drivingBone {boneName=pelvis}, planeTransform {location=<0.025 -10.6023 0.9801>, rotation=<rot -1.4222 7.4184 -79.2075>}}]
   }
   n4 = Control Rig!(Source=Component To Local(ComponentPose=@n3), Alpha, Is crouching:bool=Get bIsCrouching()) -> (Pose) {
    alphaBoolBlend {blendOption=Linear}
    controlRigAssetReference {blueprintRigClass=&/M_MED_MustardToast_CtrlRig.M_MED_MustardToast_CtrlRig_C}
   }
   n5 = "Transform (Modify) Bone - Bone: dyn_vest_strap_r"!(
    ComponentPose=Local To Component(LocalPose=@n4)
    bAlphaBoolEnabled:bool=Get bIsJumping()
    Translation="0,0,0"
    Rotation="15,0,0"
    Scale="1,1,1"
   ) -> (Pose) {
    alphaInputType=Bool
    rotationMode=BMM_Additive
    rotationSpace=BCS_BoneSpace
    translationMode=BMM_Additive
    alphaBoolBlend {blendInTime=0.5, blendOption=Linear, blendOutTime=0.5}
    boneToModify {boneName=dyn_vest_strap_r}
    rotation {pitch=15}
   }
   n6 = "Transform (Modify) Bone - Bone: dyn_vest_strap_l"!(
    ComponentPose=@n5
    bAlphaBoolEnabled:bool=Get crouching corr() -> (crouching corr:bool)
    Translation="1.842,1.020,2.847"
    Rotation="-15.220,-2.591,10.318"
    Scale="1,1,1"
   ) -> (Pose) {
    alphaInputType=Bool
    rotationMode=BMM_Additive
    rotationSpace=BCS_BoneSpace
    translationMode=BMM_Additive
    translationSpace=BCS_BoneSpace
    alphaBoolBlend {blendInTime=0.4, blendOption=Linear, blendOutTime=0.3}
    boneToModify {boneName=dyn_vest_strap_l}
    rotation=<rot -15.22 -2.591 10.318>
    translation=<1.842 1.02 2.847>
   }
   n7 = "Transform (Modify) Bone - Bone: dyn_vest_strap_r"(ComponentPose=@n6, bAlphaBoolEnabled=Get bIsSkydiveFloating(), Translation="0,0,0", Rotation="-10,0,0", Scale="1,1,1") {
    alphaInputType=Bool
    rotationMode=BMM_Additive
    rotationSpace=BCS_BoneSpace
    translationMode=BMM_Additive
    alphaBoolBlend {blendInTime=0.5, blendOption=Linear, blendOutTime=0.5}
    boneToModify {boneName=dyn_vest_strap_r}
    rotation {pitch=-10}
   }
   Save cached pose 'main cache'(Pose=Component To Local(ComponentPose=@n7)) {cacheName=main cache}
   n8 = "Transform (Modify) Bone - Bone: strap_L"!(ComponentPose=Local To Component(LocalPose=Use cached pose 'main cache'()), Alpha=1, Translation="1,0,0.5", Rotation="0,0,0") -> (Pose) {
    rotationMode=BMM_Additive
    rotationSpace=BCS_BoneSpace
    translationMode=BMM_Additive
    translationSpace=BCS_BoneSpace
    alphaBoolBlend {blendInTime=0.5, blendOption=Linear, blendOutTime=0.5}
    boneToModify {boneName=strap_L}
    translation {x=1, z=0.5}
   }
   n9 = "Transform (Modify) Bone - Bone: strap_L"(ComponentPose=@n8, Alpha=1, Translation="3,1.5,-1", Rotation="0,0,0") {
    rotationMode=BMM_Additive
    rotationSpace=BCS_BoneSpace
    translationMode=BMM_Additive
    translationSpace=BCS_BoneSpace
    alphaBoolBlend {blendInTime=0.5, blendOption=Linear, blendOutTime=0.5}
    boneToModify {boneName=strap_L}
    translation=<3 1.5 -1>
   }
   n10 = "Transform (Modify) Bone - Bone: strap_R"!(ComponentPose=@n9, Alpha=1, Translation="-3,0,0", Rotation="0,0,0") -> (Pose) {
    rotationMode=BMM_Additive
    rotationSpace=BCS_BoneSpace
    translationMode=BMM_Additive
    translationSpace=BCS_BoneSpace
    alphaBoolBlend {blendInTime=0.5, blendOption=Linear, blendOutTime=0.5}
    boneToModify {boneName=strap_R}
    translation {x=-3}
   }
   Save cached pose 'common cache'(Pose=Component To Local(ComponentPose=@n10)) {cacheName=common cache}
   n11 = "Transform (Modify) Bone - Bone: strap_L"(ComponentPose=Local To Component(LocalPose=Use cached pose 'main cache'()), Alpha=1, Translation="3.5,0,0", Rotation="0,0,0") {
    rotationMode=BMM_Additive
    rotationSpace=BCS_BoneSpace
    translationMode=BMM_Additive
    translationSpace=BCS_BoneSpace
    alphaBoolBlend {blendInTime=0.5, blendOption=Linear, blendOutTime=0.5}
    boneToModify {boneName=strap_L}
    translation {x=3.5}
   }
   n12 = "Transform (Modify) Bone - Bone: strap_R"(ComponentPose=@n11, Alpha=1, Translation="-3.5,0,0", Rotation="0,0,0") {
    rotationMode=BMM_Additive
    rotationSpace=BCS_BoneSpace
    translationMode=BMM_Additive
    translationSpace=BCS_BoneSpace
    alphaBoolBlend {blendInTime=0.5, blendOption=Linear, blendOutTime=0.5}
    boneToModify {boneName=strap_R}
    translation {x=-3.5}
   }
   Save cached pose 'skydiving'(Pose=Component To Local(ComponentPose=@n12)) {cacheName=skydiving}
  ]}
 EventGraph {nodes [
   n1 = Event BlueprintUpdateAnimation()
   > Set Delta Time X(Delta Time X=@n1.DeltaTimeX)
   > n2 = Sequence!(execute) -> (then_0, then_1, then_2:Exec)
   n3 = Branch(execute=@n2.then_1, Condition=Get bIsCrouchSprinting())
   > Do Once(Reset=@n4.then)
   > Set crouching corr(crouching corr:bool=true) -> (Output_Get:bool)
   > Delay(Duration=0.4)
   > Set crouching corr(crouching corr=false)
   n4 = Set crouching corr(execute=@n3.else, crouching corr=false)
   n5 = GetOwnerSkeletalMeshComponent()
   n6 = BreakTransform!(InTransform=GetDeltaTransformFromRefPose(self=@n5, BoneName=clavicle_l, BaseName=spine_05)) -> (Location, Rotation_Roll:float, Rotation_Pitch:float, Rotation_Yaw:float, Scale)
   n7 = MapRangeClamped(Value=@n6.Rotation_Pitch, InRangeA=27, InRangeB=45)
   n8 = MapRangeClamped(Value=@n6.Rotation_Roll, InRangeA=0, InRangeB=14)
   Set Clavicle L Pitch Roll(execute=@n2.then_2, Clavicle L Pitch Roll:float=MapRangeClamped(Value=float + float(A=@n8, B=@n7), InRangeA=1.2, InRangeB=2)) -> (Output_Get:float)
   > Set Clavicle L Pitch(Clavicle L Pitch:double=@n7) -> (Output_Get:double)
   n9 = GetDeltaTransformFromRefPose(self=@n5, BoneName=clavicle_r, BaseName=spine_05)
   > Set Clavicle R Pitch(Clavicle R Pitch:double=MapRangeClamped(Value=BreakTransform(InTransform=@n9).Rotation_Pitch, InRangeA=27, InRangeB=45)) -> (Output_Get:double)
   n10 = Gravity Override State Function DEPRECATED(
    execute=@n2.then_0
    CharacterPartAnimInstance=Self-Reference()
    GameTimeInSeconds=GetGameTimeInSeconds()
    GravityOverrideParams=Get Gravity Override Params()
    RBAN Params=Get RBAN Params()
   )
   > Set Gravity Override(Gravity Override=@n10)
  ]}
]
variables [
 Clavicle L Pitch {type=number}
 Clavicle L Pitch Roll {type=number}
 Clavicle R Pitch {type=number}
 crouching corr {type=boolean}
 Delta Time X {type=number}
 Gravity Override {type=Vector}
 Gravity Override Params {type=GravityOverrideParamsStruct, default {skydive_Parachuting {gravityOverride=<0 -800 -1500>, windFrequency=<1.5 5 6>}}}
 RBAN Params {
  type=RBANParamsStruct
  default {
   dBNO {simSpaceSettings {
     maxAngularAcceleration=1000
     maxAngularVelocity=3
     maxLinearAcceleration=300
     maxLinearVelocity=700
     velocityScaleZ=0
     externalLinearDragV=<0 0.5 0>
    }}
   dBNO_Carried {simSpaceSettings {
     maxAngularAcceleration=1000
     maxAngularVelocity=3
     maxLinearAcceleration=300
     maxLinearVelocity=700
     velocityScaleZ=0
     externalLinearDragV=<0 0.5 0>
    }}
   emoteOrMelee {simSpaceSettings {
     maxAngularAcceleration=1000
     maxAngularVelocity=3
     maxLinearAcceleration=300
     maxLinearVelocity=700
     velocityScaleZ=0
     externalLinearDragV=<0 0.5 0>
    }}
   front_End {simSpaceSettings {
     maxAngularAcceleration=1000
     maxAngularVelocity=3
     maxLinearAcceleration=300
     maxLinearVelocity=700
     velocityScaleZ=0
     externalLinearDragV=<0 0.5 0>
    }}
   onGround_Crouch_Moving {simSpaceSettings {
     maxAngularAcceleration=1000
     maxAngularVelocity=3
     maxLinearAcceleration=300
     maxLinearVelocity=700
     velocityScaleZ=0
     externalLinearDragV=<0 0.5 0>
    }}
   onGround_Falling {simSpaceSettings {
     maxAngularAcceleration=1000
     maxAngularVelocity=3
     maxLinearAcceleration=300
     maxLinearVelocity=700
     velocityScaleZ=0
     externalLinearDragV=<0 0.5 0>
    }}
   onGround_Moving {simSpaceSettings {
     maxAngularAcceleration=1000
     maxAngularVelocity=3
     maxLinearAcceleration=300
     maxLinearVelocity=700
     velocityScaleZ=0
     externalLinearDragV=<0 0.5 0>
    }}
   onGround_Standing {simSpaceSettings {
     maxAngularAcceleration=1000
     maxAngularVelocity=3
     maxLinearAcceleration=300
     maxLinearVelocity=700
     velocityScaleZ=0
     externalLinearDragV=<0 0.5 0>
    }}
   rail_Grinding {simSpaceSettings {
     maxAngularAcceleration=1000
     maxAngularVelocity=3
     maxLinearAcceleration=300
     maxLinearVelocity=700
     velocityScaleZ=0
     worldAlpha=0.1
     externalLinearDragV=<0 0.5 0>
    }}
   riding_Moving {simSpaceSettings {
     maxAngularAcceleration=1000
     maxAngularVelocity=3
     maxLinearAcceleration=300
     maxLinearVelocity=700
     velocityScaleZ=0
     externalLinearDragV=<0 0.5 0>
    }}
   riding_Moving_Targeting {simSpaceSettings {
     maxAngularAcceleration=1000
     maxAngularVelocity=3
     maxLinearAcceleration=300
     maxLinearVelocity=700
     velocityScaleZ=0
     externalLinearDragV=<0 0.5 0>
    }}
   riding_Standing {simSpaceSettings {
     maxAngularAcceleration=1000
     maxAngularVelocity=3
     maxLinearAcceleration=300
     maxLinearVelocity=700
     velocityScaleZ=0
     externalLinearDragV=<0 0.5 0>
    }}
   skydive_DivingDown {simSpaceSettings {
     maxAngularAcceleration=1000
     maxAngularVelocity=3
     maxLinearAcceleration=300
     maxLinearVelocity=700
     velocityScaleZ=0
     externalLinearDragV=<0 0.5 0>
    }}
   skydive_FloatingDown {
    componentLinearAccScale=<0 0.5 0>
    componentLinearVelScale=<0 0.5 0>
    simSpaceSettings {
     maxAngularAcceleration=1000
     maxAngularVelocity=3
     maxLinearAcceleration=300
     maxLinearVelocity=700
     velocityScaleZ=0
     externalLinearDragV=<0 0.5 0>
    }
   }
   skydive_GoingUp {simSpaceSettings {
     maxAngularAcceleration=1000
     maxAngularVelocity=3
     maxLinearAcceleration=300
     maxLinearVelocity=700
     velocityScaleZ=0
     externalLinearDragV=<0 0.5 0>
    }}
   skydive_Parachuting {
    componentLinearAccScale=<0 0.1 0.3>
    componentLinearVelScale=<0 0.1 0.3>
    simSpaceSettings {
     maxAngularAcceleration=1000
     maxAngularVelocity=3
     maxLinearAcceleration=300
     maxLinearVelocity=700
     velocityScaleZ=0
     externalLinearDragV=<0 0.5 0>
    }
   }
   skydive_Surfing {simSpaceSettings {
     maxAngularAcceleration=1000
     maxAngularVelocity=3
     maxLinearAcceleration=300
     maxLinearVelocity=700
     velocityScaleZ=0
     externalLinearDragV=<0 0.5 0>
    }}
   swimming {simSpaceSettings {
     maxAngularAcceleration=1000
     maxAngularVelocity=3
     maxLinearAcceleration=300
     maxLinearVelocity=700
     velocityScaleZ=0
     externalLinearDragV=<0 0.5 0>
    }}
   tactical_Sprint {simSpaceSettings {
     maxAngularAcceleration=1000
     maxAngularVelocity=3
     maxLinearAcceleration=300
     maxLinearVelocity=700
     velocityScaleZ=0
     externalLinearDragV=<0 0.5 0>
    }}
   ziplining {simSpaceSettings {
     maxAngularAcceleration=1000
     maxAngularVelocity=3
     maxLinearAcceleration=300
     maxLinearVelocity=700
     velocityScaleZ=0
     externalLinearDragV=<0 0.5 0>
    }}
  }
 }
]
