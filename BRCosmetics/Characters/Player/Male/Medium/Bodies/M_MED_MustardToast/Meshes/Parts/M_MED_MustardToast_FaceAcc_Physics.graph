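# Physics asset for the FaceAcc part of the MustardToast medium male player body, simulating its dynamic hair bones with capsule bodies.
# Six chains (front, back, two sides, two back-mid) hang from MSTR bones through tight swing and twist limits.
# Cross-constraints, mostly angularly free, tie side chains to back-mid and back-mid to back at selected segments, coupling neighboring strands.
# Head, neck, upper spine, clavicle and upper-arm capsules provide body shapes.
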PhysicsAsset /BRCosmetics/Characters/Player/Male/Medium/Bodies/M_MED_MustardToast/Meshes/Parts/M_MED_MustardToast_FaceAcc_Physics
bodies {
 clavicle_phs_l [{shapeType=Capsule, length=7, radius=9.5, shapeName=clavicle_r_capsule, center=<9.7 0 -0.7>, rotation=<rot 70 -0.0003 -0.0004>}]
 clavicle_phs_r [{shapeType=Capsule, length=7, radius=9.5, shapeName=clavicle_r_capsule, center=<-9.7 0 0.7>, rotation=<rot 70 -0.0003 -0.0004>}]
 dyn_hair_01_L [{shapeType=Capsule, length=1, radius=2.5, shapeName=dyn_hair_01_L_capsule, center=<1 0 0>, rotation=<rot -0.0011 -89.9984 90.0004>}]
 dyn_hair_01_R [{shapeType=Capsule, length=1, radius=2.5, shapeName=dyn_hair_01_R_capsule, center=<1 0 0>, rotation=<rot 0.0012 -90.0017 89.9998>}]
 dyn_hair_02_L [{shapeType=Capsule, length=1, radius=2.5, shapeName=dyn_hair_02_L_capsule, center=<1 0 0>, rotation=<rot -0.0008 -89.9982 90.0004>}]
 dyn_hair_02_R [{shapeType=Capsule, length=1, radius=2.5, shapeName=dyn_hair_02_R_capsule, center=<1 0 0>, rotation=<rot 0.0008 -90.0019 89.9999>}]
 dyn_hair_03_L [{shapeType=Capsule, length=1, radius=2.5, shapeName=dyn_hair_03_L_capsule, center=<1 0 0>, rotation=<rot -0.0004 -89.998 90.0004>}]
 dyn_hair_03_R [{shapeType=Capsule, length=1, radius=2.5, shapeName=dyn_hair_03_R_capsule, center=<1 0 0>, rotation=<rot 0.0005 -90.002 89.9998>}]
 dyn_hair_04_L [{shapeType=Capsule, length=1, radius=2.5, shapeName=dyn_hair_04_L_capsule, center=<2.5112 0.0001 0>, rotation=<rot -0.0001 -89.9979 90.0004>}]
 dyn_hair_04_R [{
   shapeType=Capsule
   length=1.5
   radius=2.5
   shapeName=dyn_hair_04_R_capsule
   center=<2.4676 -0.0001 0>
   rotation=<rot 0.0001 -90.0021 89.9998>
  }]
 dyn_hair_05_L [{shapeType=Capsule, length=2, radius=2.5, shapeName=dyn_hair_05_L_capsule, center=<2.7901 -0.442 0>, rotation=<rot -0.0003 40.003 89.9995>}]
 dyn_hair_05_R [{shapeType=Capsule, length=2.5, radius=3, shapeName=dyn_hair_05_R_capsule, center=<3.0082 1.697 0>, rotation=<rot 0 29.9997 90.0005>}]
 dyn_hair_MSTR_L [{shapeType=Capsule, length=0, radius=1, shapeName=dyn_hair_MSTR_L_capsule}]
 dyn_hair_MSTR_R [{shapeType=Capsule, length=0, radius=1, shapeName=dyn_hair_MSTR_R_capsule}]
 dyn_hair_bk_01 [{shapeType=Capsule, length=1, radius=2.5, shapeName=dyn_hair_bk_01_capsule, center=<1 0 0>, rotation=<rot 0.0001 -90.0001 90.002>}]
 dyn_hair_bk_02 [{shapeType=Capsule, length=1, radius=2.5, shapeName=dyn_hair_bk_02_capsule, center=<1 0 0>, rotation=<rot 0 -90.0001 90.0021>}]
 dyn_hair_bk_03 [{shapeType=Capsule, length=1, radius=2.5, shapeName=dyn_hair_bk_03_capsule, center=<1 0 0>, rotation=<rot 0 -90.0002 90.0021>}]
 dyn_hair_bk_04 [{shapeType=Capsule, length=1, radius=2.5, shapeName=dyn_hair_bk_04_capsule, center=<1 0 0>, rotation=<rot 0 -90.0002 90.0021>}]
 dyn_hair_bk_05 [{shapeType=Capsule, length=2.5, radius=3.5, shapeName=dyn_hair_bk_05_capsule, center=<1.5 0 -0.0001>, rotation=<rot 0.0054 179.9996 89.9999>}]
 dyn_hair_bk_MSTR [{shapeType=Capsule, length=0, radius=1, shapeName=dyn_hair_bk_MSTR_capsule}]
 dyn_hair_bk_mid_01_L [{shapeType=Capsule, length=1, radius=2.5, shapeName=dyn_hair_bk_mid_01_L_capsule, center=<1 0 0>, rotation=<rot -0.0008 -89.9988 90.0014>}]
 dyn_hair_bk_mid_01_R [{shapeType=Capsule, length=1, radius=2.5, shapeName=dyn_hair_bk_mid_01_R_capsule, center=<1 0 0>, rotation=<rot 0.0007 -90.0013 90.0014>}]
 dyn_hair_bk_mid_02_L [{shapeType=Capsule, length=1, radius=2.5, shapeName=dyn_hair_bk_mid_02_L_capsule, center=<1 0 0>, rotation=<rot -0.0006 -89.9987 90.0014>}]
 dyn_hair_bk_mid_02_R [{shapeType=Capsule, length=1, radius=2.5, shapeName=dyn_hair_bk_mid_02_R_capsule, center=<1 0 0>, rotation=<rot 0.0006 -90.0014 90.0014>}]
 dyn_hair_bk_mid_03_L [{shapeType=Capsule, length=1, radius=2.5, shapeName=dyn_hair_bk_mid_03_L_capsule, center=<1 0 0>, rotation=<rot -0.0003 -89.9986 90.0014>}]
 dyn_hair_bk_mid_03_R [{shapeType=Capsule, length=1, radius=2.5, shapeName=dyn_hair_bk_mid_03_R_capsule, center=<1 0 0>, rotation=<rot 0.0005 -90.0014 90.0014>}]
 dyn_hair_bk_mid_04_L [{shapeType=Capsule, length=1, radius=2.5, shapeName=dyn_hair_bk_mid_04_L_capsule, center=<1 0 0>, rotation=<rot -0.0002 -89.9985 90.0014>}]
 dyn_hair_bk_mid_04_R [{shapeType=Capsule, length=1, radius=2.5, shapeName=dyn_hair_bk_mid_04_R_capsule, center=<1 0 0>, rotation=<rot 0.0003 -90.0015 90.0014>}]
 dyn_hair_bk_mid_05_L [{shapeType=Capsule, length=2, radius=3, shapeName=dyn_hair_bk_mid_05_L_capsule, center=<0.5 0 0>, rotation=<rot 0.0026 -179.9973 89.9997>}]
 dyn_hair_bk_mid_05_R [{
   shapeType=Capsule
   length=4.5
   radius=3.5
   shapeName=dyn_hair_bk_mid_05_R_capsule
   center=<-0.3231 0.0001 0.0001>
   rotation=<rot -0.0022 -0.0026 90.0003>
  }]
 dyn_hair_bk_mid_MSTR_L [{shapeType=Capsule, length=0, radius=1, shapeName=dyn_hair_bk_mid_MSTR_L_capsule}]
 dyn_hair_bk_mid_MSTR_R [{shapeType=Capsule, length=0, radius=1, shapeName=dyn_hair_bk_mid_MSTR_R_capsule}]
 dyn_hair_fr_01 [{shapeType=Capsule, length=1, radius=2.5, shapeName=dyn_hair_fr_01_capsule, center=<1 0 0>, rotation=<rot 0.0007 -90.0021 89.9993>}]
 dyn_hair_fr_02 [{shapeType=Capsule, length=1, radius=2.5, shapeName=dyn_hair_fr_02_capsule, center=<1 0 0>, rotation=<rot 0.0009 -90.0021 89.9997>}]
 dyn_hair_fr_03 [{shapeType=Capsule, length=1, radius=2.5, shapeName=dyn_hair_fr_03_capsule, center=<1 0 0>, rotation=<rot 0.0009 -90.0021 90.0002>}]
 dyn_hair_fr_MSTR [{shapeType=Capsule, length=0, radius=1, shapeName=dyn_hair_fr_MSTR_capsule}]
 head [{shapeType=Capsule, length=7, radius=4, shapeName=head_capsule, center=<4.8042 2.9191 0>, rotation=<rot -90 -165.9638 75.9634>}]
 neck_01 [{shapeType=Capsule, length=1.5, radius=5, shapeName=neck_01_capsule, center=<1.4708 0 0>, rotation=<rot -90 0.0024 -0.0029>}]
 neck_02 [{shapeType=Capsule, length=3, radius=4.5, shapeName=neck_02_capsule, center=<2.8434 0 0>, rotation=<rot -90 -10.0249 10.0247>}]
 spine_05 [{shapeType=Capsule, length=7.5, radius=10, shapeName=spine_05_capsule, center=<4.0453 -0.7367 0>}]
 upperarm_l [{shapeType=Capsule, length=21.5, radius=5, shapeName=upperarm_r_capsule, center=<11.5 0 0>, rotation=<rot 90 -174.949 -174.9495>}]
 upperarm_r [{shapeType=Capsule, length=25, radius=5, shapeName=upperarm_r_capsule, center=<-11.5 0 0>, rotation=<rot 90 -90 -90.0006>}]
}
constraints [
 {bone1Name=dyn_hair_01_L, bone2Name=dyn_hair_MSTR_L, swing1LimitDegrees=25, swing2LimitDegrees=15, twistLimitDegrees=2.5}
 {bone1Name=dyn_hair_02_L, bone2Name=dyn_hair_01_L, swing1LimitDegrees=25, swing2LimitDegrees=15, twistLimitDegrees=2.5}
 {bone1Name=dyn_hair_03_L, bone2Name=dyn_hair_02_L, swing1LimitDegrees=10, swing2LimitDegrees=25, twistLimitDegrees=2.5}
 {bone1Name=dyn_hair_04_L, bone2Name=dyn_hair_03_L, swing1LimitDegrees=10, swing2LimitDegrees=25, twistLimitDegrees=2.5}
 {bone1Name=dyn_hair_05_L, bone2Name=dyn_hair_04_L, swing1LimitDegrees=10, swing2LimitDegrees=25, twistLimitDegrees=2.5}
 {bone1Name=dyn_hair_01_R, bone2Name=dyn_hair_MSTR_R, swing1LimitDegrees=25, swing2LimitDegrees=15, twistLimitDegrees=2.5}
 {bone1Name=dyn_hair_02_R, bone2Name=dyn_hair_01_R, swing1LimitDegrees=25, swing2LimitDegrees=15, twistLimitDegrees=2.5}
 {bone1Name=dyn_hair_03_R, bone2Name=dyn_hair_02_R, swing1LimitDegrees=10, swing2LimitDegrees=25, twistLimitDegrees=2.5}
 {bone1Name=dyn_hair_04_R, bone2Name=dyn_hair_03_R, swing1LimitDegrees=10, swing2LimitDegrees=25, twistLimitDegrees=2.5}
 {bone1Name=dyn_hair_05_R, bone2Name=dyn_hair_04_R, swing1LimitDegrees=10, swing2LimitDegrees=25, twistLimitDegrees=2.5}
 {bone1Name=dyn_hair_bk_01, bone2Name=dyn_hair_bk_MSTR, swing1LimitDegrees=10, swing2LimitDegrees=15, twistLimitDegrees=2.5}
 {bone1Name=dyn_hair_bk_02, bone2Name=dyn_hair_bk_01, swing1LimitDegrees=10, swing2LimitDegrees=15, twistLimitDegrees=2.5}
 {bone1Name=dyn_hair_bk_03, bone2Name=dyn_hair_bk_02, swing1LimitDegrees=10, swing2LimitDegrees=20, twistLimitDegrees=2.5}
 {bone1Name=dyn_hair_bk_04, bone2Name=dyn_hair_bk_03, swing1LimitDegrees=10, swing2LimitDegrees=25, twistLimitDegrees=2.5}
 {bone1Name=dyn_hair_bk_05, bone2Name=dyn_hair_bk_04, swing1LimitDegrees=10, swing2LimitDegrees=25, twistLimitDegrees=2.5}
 {bone1Name=dyn_hair_bk_mid_01_L, bone2Name=dyn_hair_bk_mid_MSTR_L, swing1LimitDegrees=25, swing2LimitDegrees=15, twistLimitDegrees=2.5}
 {bone1Name=dyn_hair_bk_mid_02_L, bone2Name=dyn_hair_bk_mid_01_L, swing1LimitDegrees=25, swing2LimitDegrees=15, twistLimitDegrees=2.5}
 {bone1Name=dyn_hair_bk_mid_03_L, bone2Name=dyn_hair_bk_mid_02_L, swing1LimitDegrees=25, swing2LimitDegrees=25, twistLimitDegrees=2.5}
 {bone1Name=dyn_hair_bk_mid_04_L, bone2Name=dyn_hair_bk_mid_03_L, swing1LimitDegrees=25, swing2LimitDegrees=25, twistLimitDegrees=2.5}
 {bone1Name=dyn_hair_bk_mid_05_L, bone2Name=dyn_hair_bk_mid_04_L, swing1LimitDegrees=25, swing2LimitDegrees=25, twistLimitDegrees=2.5}
 {bone1Name=dyn_hair_bk_mid_01_R, bone2Name=dyn_hair_bk_mid_MSTR_R, swing1LimitDegrees=25, swing2LimitDegrees=15, twistLimitDegrees=2.5}
 {bone1Name=dyn_hair_bk_mid_02_R, bone2Name=dyn_hair_bk_mid_01_R, swing1LimitDegrees=25, swing2LimitDegrees=15, twistLimitDegrees=2.5}
 {bone1Name=dyn_hair_bk_mid_03_R, bone2Name=dyn_hair_bk_mid_02_R, swing1LimitDegrees=25, swing2LimitDegrees=25, twistLimitDegrees=2.5}
 {bone1Name=dyn_hair_bk_mid_04_R, bone2Name=dyn_hair_bk_mid_03_R, swing1LimitDegrees=25, swing2LimitDegrees=25, twistLimitDegrees=2.5}
 {bone1Name=dyn_hair_bk_mid_05_R, bone2Name=dyn_hair_bk_mid_04_R, swing1LimitDegrees=25, swing2LimitDegrees=25, twistLimitDegrees=2.5}
 {bone1Name=dyn_hair_fr_01, bone2Name=dyn_hair_fr_MSTR, swing1LimitDegrees=10, swing2LimitDegrees=10, twistLimitDegrees=2.5}
 {bone1Name=dyn_hair_fr_02, bone2Name=dyn_hair_fr_01, swing1LimitDegrees=25, swing2LimitDegrees=25, twistLimitDegrees=2.5}
 {bone1Name=dyn_hair_fr_03, bone2Name=dyn_hair_fr_02, swing1LimitDegrees=25, swing2LimitDegrees=25, twistLimitDegrees=2.5}
 {
  bone1Name=dyn_hair_bk_mid_03_L
  bone2Name=dyn_hair_bk_03
  swing1LimitDegrees=45
  swing1Motion=Free
  swing2LimitDegrees=45
  swing2Motion=Free
  twistLimitDegrees=45
  twistMotion=Free
 }
 {
  bone1Name=dyn_hair_bk_mid_03_R
  bone2Name=dyn_hair_bk_03
  swing1LimitDegrees=45
  swing1Motion=Free
  swing2LimitDegrees=45
  swing2Motion=Free
  twistLimitDegrees=45
  twistMotion=Free
 }
 {
  bone1Name=dyn_hair_bk_mid_05_R
  bone2Name=dyn_hair_bk_05
  swing1LimitDegrees=45
  swing1Motion=Free
  swing2LimitDegrees=45
  swing2Motion=Free
  twistLimitDegrees=45
  twistMotion=Free
 }
 {
  bone1Name=dyn_hair_03_R
  bone2Name=dyn_hair_bk_mid_03_R
  swing1LimitDegrees=45
  swing1Motion=Free
  swing2LimitDegrees=25
  swing2Motion=Free
  twistLimitDegrees=45
  twistMotion=Free
 }
 {
  bone1Name=dyn_hair_03_L
  bone2Name=dyn_hair_bk_mid_03_L
  swing1LimitDegrees=45
  swing1Motion=Free
  swing2LimitDegrees=45
  swing2Motion=Free
  twistLimitDegrees=45
  twistMotion=Free
 }
 {bone1Name=dyn_hair_05_R, bone2Name=dyn_hair_bk_mid_05_R, swing1LimitDegrees=0, swing2LimitDegrees=0, twistLimitDegrees=45, twistMotion=Free}
 {
  bone1Name=dyn_hair_bk_mid_05_L
  bone2Name=dyn_hair_bk_05
  swing1LimitDegrees=45
  swing1Motion=Free
  swing2LimitDegrees=45
  swing2Motion=Free
  twistLimitDegrees=45
  twistMotion=Free
 }
 {
  bone1Name=dyn_hair_05_L
  bone2Name=dyn_hair_bk_mid_05_L
  swing1LimitDegrees=45
  swing1Motion=Free
  swing2LimitDegrees=45
  swing2Motion=Free
  twistLimitDegrees=45
  twistMotion=Free
 }
 {
  bone1Name=dyn_hair_bk_mid_02_R
  bone2Name=dyn_hair_bk_02
  swing1LimitDegrees=45
  swing1Motion=Free
  swing2LimitDegrees=45
  swing2Motion=Free
  twistLimitDegrees=45
  twistMotion=Free
 }
 {
  bone1Name=dyn_hair_bk_mid_02_L
  bone2Name=dyn_hair_bk_02
  swing1LimitDegrees=45
  swing1Motion=Free
  swing2LimitDegrees=45
  swing2Motion=Free
  twistLimitDegrees=45
  twistMotion=Free
 }
]
