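# Post-animation Blueprint for the MustardToast face accessory, adding secondary motion to hair and medallions.
# The copied parent pose gets a medallion bone constraint, AnimDynamics and hair offsets, then branches into cached default, crouching and skydiving poses.
# Neck-relative RigidBody nodes simulate these, with a separate skydiving physics asset, before final hair tweaks.
# The EventGraph computes state-based gravity and RBAN settings, plus head and clavicle corrective values.
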AnimBlueprint /BRCosmetics/Characters/Player/Male/Medium/Bodies/M_MED_MustardToast/Meshes/Parts/M_MED_MustardToast_FaceAcc_AnimBP
target_skeleton=&/M_MED_MustardToast_FaceAcc_Skeleton
graphs [
 AnimGraph {nodes [
   n1 = Use cached pose 'common cache'()
   n2 = Use cached pose 'crouching'()
   n3 = Get bIsCrouching()
   n4 = RigidBody!(
    ComponentPose=Local To Component(LocalPose=Blend Poses by bool(BlendPose_0=@n2, BlendPose_1=@n1, bActiveValue=@n3) {blendType=Linear})
    Alpha
    OverrideWorldGravity:Vector="0,0,0"
    ComponentLinearAccScale:Vector="0.5,0.25,0.5"
    ComponentLinearVelScale:Vector="0.5,0.25,0.5"
   ) -> (Pose) {
    bClampLinearTranslationLimitToRefPose=true
    bOverrideWorldGravity=true
    overridePhysicsAsset=&/M_MED_MustardToast_FaceAcc_Physics
    simulationSpace=BaseBoneSpace
    alphaBoolBlend {blendOption=Linear}
    baseBoneRef {boneName=neck_02}
    componentAppliedLinearAccClamp=<1000 1000 1000>
   }
   n5 = Component To Local(ComponentPose=@n4)
   n6 = Use cached pose 'skydiving'()
   n7 = Get bIsSkyDiving()
   n8 = "Transform (Modify) Bone - Bone: dyn_hair_bk_mid_04_R"!(
    ComponentPose=Local To Component(LocalPose=Blend Poses by bool(BlendPose_0=@n6, BlendPose_1=@n5, bActiveValue=@n7) {blendType=Linear})
    Alpha=1
    Translation="0,0,0"
    Rotation="-20,0,0"
   ) -> (Pose) {
    rotationMode=BMM_Additive
    rotationSpace=BCS_BoneSpace
    translationMode=BMM_Additive
    translationSpace=BCS_BoneSpace
    alphaBoolBlend {blendOption=Linear}
    boneToModify {boneName=dyn_hair_bk_mid_04_R}
    rotation {pitch=-20}
   }
   n9 = "Transform (Modify) Bone - Bone: dyn_hair_bk_mid_05_R"!(ComponentPose=@n8, Alpha=1, Translation="0,0,0", Rotation="-30,0,0") -> (Pose) {
    rotationMode=BMM_Additive
    rotationSpace=BCS_BoneSpace
    translationMode=BMM_Additive
    translationSpace=BCS_BoneSpace
    alphaBoolBlend {blendOption=Linear}
    boneToModify {boneName=dyn_hair_bk_mid_05_R}
    rotation {pitch=-30}
   }
   n10 = "Transform (Modify) Bone - Bone: dyn_hair_bk_02"!(ComponentPose=@n9, Alpha=1, Translation="0,0,0", Rotation="10,0,0") -> (Pose) {
    rotationMode=BMM_Additive
    rotationSpace=BCS_BoneSpace
    translationMode=BMM_Additive
    translationSpace=BCS_BoneSpace
    alphaBoolBlend {blendOption=Linear}
    boneToModify {boneName=dyn_hair_bk_02}
    rotation {pitch=10}
   }
   n11 = "Transform (Modify) Bone - Bone: dyn_hair_04_R"!(ComponentPose=@n10, Alpha=1, Translation="0,0,0", Rotation="-10,0,0") -> (Pose) {
    rotationMode=BMM_Additive
    rotationSpace=BCS_BoneSpace
    translationMode=BMM_Additive
    translationSpace=BCS_BoneSpace
    alphaBoolBlend {blendOption=Linear}
    boneToModify {boneName=dyn_hair_04_R}
    rotation {pitch=-10}
   }
   n12 = "Transform (Modify) Bone - Bone: dyn_hair_04_R"(ComponentPose=@n11, Alpha=1, Translation="0,0,0", Rotation="20,0,0") {
    rotationMode=BMM_Additive
    rotationSpace=BCS_BoneSpace
    translationMode=BMM_Additive
    translationSpace=BCS_BoneSpace
    alphaBoolBlend {blendOption=Linear}
    boneToModify {boneName=dyn_hair_04_R}
    rotation {pitch=20}
   }
   n13 = "Transform (Modify) Bone - Bone: dyn_hair_05_R"!(ComponentPose=@n12, Alpha=1, Translation="0,0,0", Rotation="-40,30,0") -> (Pose) {
    rotationMode=BMM_Additive
    rotationSpace=BCS_BoneSpace
    translationMode=BMM_Additive
    translationSpace=BCS_BoneSpace
    alphaBoolBlend {blendOption=Linear}
    boneToModify {boneName=dyn_hair_05_R}
    rotation {pitch=-40, yaw=30}
   }
   n14 = "Transform (Modify) Bone - Bone: dyn_hair_bk_mid_03_R"!(ComponentPose=@n13, Alpha=1, Translation="0,0,0", Rotation="-10,0,0") -> (Pose) {
    rotationMode=BMM_Additive
    rotationSpace=BCS_BoneSpace
    translationMode=BMM_Additive
    translationSpace=BCS_BoneSpace
    alphaBoolBlend {blendOption=Linear}
    boneToModify {boneName=dyn_hair_bk_mid_03_R}
    rotation {pitch=-10}
   }
   Output Pose(Result=Component To Local(ComponentPose=@n14)) {layerGroup=None}
   n15 = "Constraint - medallions_CST (PPPPPP)"!(
    ComponentPose=Local To Component(LocalPose=Copy Pose From Mesh() {bUseAttachedParent=true})
    Alpha=1
    ConstraintWeights_0=0.0009
    ConstraintWeights_1:float=0.0101
    ConstraintWeights_2:float=0.1811
    ConstraintWeights_3:float=0.4087
    ConstraintWeights_4:float=0.342
    ConstraintWeights_5:float=0.0572
   ) -> (Pose) {
    alphaBoolBlend {blendOption=Linear}
    boneToModify {boneName=medallions_CST}
    constraintSetup [
     {targetBone {boneName=spine_01}}
     {targetBone {boneName=spine_02}}
     {targetBone {boneName=spine_03}}
     {targetBone {boneName=spine_04}}
     {targetBone {boneName=spine_05}}
     {targetBone {boneName=clavicle_l}}
    ]
    constraintWeights [1,1,1,1,1,1]
   }
   n16 = Anim Dynamics - Bone: dyn_medallion_01_R!(ComponentPose=@n15, Alpha=1, GravityOverride:Vector="0,0,-980") -> (Pose) {
    angularBiasOverride=1
    angularDampingOverride=0.9999
    bOverrideAngularBias=true
    bOverrideAngularDamping=true
    bUseGravityOverride=true
    bUseSphericalLimits=true
    simulationSpace=RootRelative
    alphaBoolBlend {blendOption=Linear}
    boundBone {boneName=dyn_medallion_01_R}
    componentAppliedLinearAccClamp=<1500 1500 1500>
    componentLinearAccScale=<0.5 0.5 0.5>
    componentLinearVelScale=<0.5 0.5 0.5>
    physicsBodyDefinitions [{
      boundBone {boneName=dyn_medallion_01_R}
      boxExtents=<10 8 4>
      constraintSetup {angularLimitsMax=<0 -5 15>, angularLimitsMin=<0 -20 -15>, angularTarget=<1 0 0>}
      localJointOffset=<1 0 0>
     }]
    sphericalLimits [{limitRadius=0, drivingBone {boneName=spine_04}, sphereLocalOffset=<0 0 0>}]
   }
   n17 = Anim Dynamics - Bone: dyn_medallion_01_L!(ComponentPose=@n16, Alpha=1, GravityOverride:Vector="0,0,-980") -> (Pose) {
    angularBiasOverride=1
    angularDampingOverride=0.9999
    bOverrideAngularBias=true
    bOverrideAngularDamping=true
    bUseGravityOverride=true
    bUseSphericalLimits=true
    simulationSpace=RootRelative
    alphaBoolBlend {blendOption=Linear}
    boundBone {boneName=dyn_medallion_01_L}
    componentAppliedLinearAccClamp=<1500 1500 1500>
    componentLinearAccScale=<0.5 0.5 0.5>
    componentLinearVelScale=<0.5 0.5 0.5>
    physicsBodyDefinitions [{
      boundBone {boneName=dyn_medallion_01_L}
      boxExtents=<15 12 9>
      constraintSetup {angularLimitsMax=<0 -5 15>, angularLimitsMin=<0 -20 -15>, angularTarget=<1 0 0>}
      localJointOffset=<1 0 0>
     }]
    sphericalLimits [{limitRadius=0, drivingBone {boneName=spine_04}, sphereLocalOffset=<0 0 0>}]
   }
   n18 = Anim Dynamics - Bone: dyn_medallion_02_L!(ComponentPose=@n17, Alpha=1, GravityOverride:Vector="0,0,-980") -> (Pose) {
    angularBiasOverride=1
    angularDampingOverride=0.9999
    bOverrideAngularBias=true
    bOverrideAngularDamping=true
    bUseGravityOverride=true
    bUseSphericalLimits=true
    simulationSpace=RootRelative
    alphaBoolBlend {blendOption=Linear}
    boundBone {boneName=dyn_medallion_02_L}
    componentAppliedLinearAccClamp=<1500 1500 1500>
    componentLinearAccScale=<0.5 0.5 0.5>
    componentLinearVelScale=<0.5 0.5 0.5>
    physicsBodyDefinitions [{
      boundBone {boneName=dyn_medallion_02_L}
      boxExtents=<10 8 4>
      constraintSetup {angularLimitsMax=<0 -5 15>, angularLimitsMin=<0 -20 -15>, angularTarget=<1 0 0>}
      localJointOffset=<1 0 0>
     }]
    sphericalLimits [{limitRadius=0, drivingBone {boneName=spine_04}, sphereLocalOffset=<0 0 0>}]
   }
   n19 = "Transform (Modify) Bone - Bone: dyn_hair_bk_MSTR"!(ComponentPose=@n18, Alpha=1, Translation="0,0,0", Rotation="70,0,0") -> (Pose) {
    rotationMode=BMM_Additive
    rotationSpace=BCS_BoneSpace
    translationMode=BMM_Additive
    translationSpace=BCS_BoneSpace
    alphaBoolBlend {blendOption=Linear}
    boneToModify {boneName=dyn_hair_bk_MSTR}
    rotation {pitch=70}
   }
   n20 = "Transform (Modify) Bone - Bone: dyn_hair_bk_mid_MSTR_R"!(ComponentPose=@n19, Alpha=1, Translation="0,0,0", Rotation="50,0,0") -> (Pose) {
    rotationMode=BMM_Additive
    rotationSpace=BCS_BoneSpace
    translationMode=BMM_Additive
    translationSpace=BCS_BoneSpace
    alphaBoolBlend {blendOption=Linear}
    boneToModify {boneName=dyn_hair_bk_mid_MSTR_R}
    rotation {pitch=50}
   }
   n21 = "Transform (Modify) Bone - Bone: dyn_hair_bk_mid_MSTR_L"!(ComponentPose=@n20, Alpha=1, Translation="0,0,0", Rotation="50,0,0") -> (Pose) {
    rotationMode=BMM_Additive
    rotationSpace=BCS_BoneSpace
    translationMode=BMM_Additive
    translationSpace=BCS_BoneSpace
    alphaBoolBlend {blendOption=Linear}
    boneToModify {boneName=dyn_hair_bk_mid_MSTR_L}
    rotation {pitch=50}
   }
   n22 = "Transform (Modify) Bone - Bone: dyn_hair_MSTR_R"!(ComponentPose=@n21, Alpha=1, Translation="0,0,0", Rotation="0,0,0") -> (Pose) {
    rotationMode=BMM_Additive
    rotationSpace=BCS_BoneSpace
    translationMode=BMM_Additive
    translationSpace=BCS_BoneSpace
    alphaBoolBlend {blendOption=Linear}
    boneToModify {boneName=dyn_hair_MSTR_R}
   }
   Save cached pose 'main cache'(Pose=Component To Local(ComponentPose=@n22)) {cacheName=main cache}
   n23 = "Transform (Modify) Bone - Bone: clavicle_phs_l"!(ComponentPose=Local To Component(LocalPose=Use cached pose 'main cache'()), Alpha=1, Translation="0,0,-3") -> (Pose) {
    translationMode=BMM_Additive
    translationSpace=BCS_BoneSpace
    alphaBoolBlend {blendOption=Linear}
    boneToModify {boneName=clavicle_phs_l}
    rotation {pitch=40}
    translation {z=-3}
   }
   n24 = "Transform (Modify) Bone - Bone: clavicle_phs_r"!(ComponentPose=@n23, Alpha=1, Translation="0,0,3") -> (Pose) {
    translationMode=BMM_Additive
    translationSpace=BCS_BoneSpace
    alphaBoolBlend {blendOption=Linear}
    boneToModify {boneName=clavicle_phs_r}
    rotation {pitch=40}
    translation {z=3}
   }
   n25 = RigidBody(ComponentPose=@n24, OverrideWorldGravity="0,0,0", ComponentLinearAccScale="0.5,0.25,0.5", ComponentLinearVelScale="0.5,0.25,0.5") {
    bClampLinearTranslationLimitToRefPose=true
    bOverrideWorldGravity=true
    overridePhysicsAsset=&/M_MED_MustardToast_FaceAcc_SkyDiving_Physics
    simulationSpace=BaseBoneSpace
    alphaBoolBlend {blendOption=Linear}
    baseBoneRef {boneName=neck_02}
    componentAppliedLinearAccClamp=<1000 1000 1000>
   }
   Save cached pose 'skydiving'(Pose=Component To Local(ComponentPose=@n25)) {cacheName=skydiving}
   n26 = "Transform (Modify) Bone - Bone: clavicle_phs_l"(ComponentPose=Local To Component(LocalPose=Use cached pose 'main cache'()), Alpha=1, Translation="0,0,-3") {
    translationMode=BMM_Additive
    translationSpace=BCS_BoneSpace
    alphaBoolBlend {blendOption=Linear}
    boneToModify {boneName=clavicle_phs_l}
    rotation {pitch=40}
    translation {z=-3}
   }
   n27 = "Transform (Modify) Bone - Bone: clavicle_phs_r"(ComponentPose=@n26, Alpha=1, Translation="0,0,3") {
    translationMode=BMM_Additive
    translationSpace=BCS_BoneSpace
    alphaBoolBlend {blendOption=Linear}
    boneToModify {boneName=clavicle_phs_r}
    rotation {pitch=40}
    translation {z=3}
   }
   n28 = "Transform (Modify) Bone - Bone: clavicle_phs_l"!(ComponentPose=@n27, Alpha=1, Translation="0,0,-3", Rotation:Rotator="0,0,0") -> (Pose) {
    rotationMode=BMM_Additive
    rotationSpace=BCS_BoneSpace
    translationMode=BMM_Additive
    translationSpace=BCS_BoneSpace
    alphaBoolBlend {blendOption=Linear}
    boneToModify {boneName=clavicle_phs_l}
    translation {z=-3}
   }
   n29 = "Transform (Modify) Bone - Bone: clavicle_phs_r"!(ComponentPose=@n28, Alpha=1, Translation="0,0,3", Rotation:Rotator="0,0,0") -> (Pose) {
    rotationMode=BMM_Additive
    rotationSpace=BCS_BoneSpace
    translationMode=BMM_Additive
    translationSpace=BCS_BoneSpace
    alphaBoolBlend {blendOption=Linear}
    boneToModify {boneName=clavicle_phs_r}
    translation {z=3}
   }
   n30 = "Transform (Modify) Bone - Bone: dyn_hair_bk_MSTR"(ComponentPose=@n29, Alpha=1, Translation="0,0,0", Rotation="1.123,39.974,-2.107") {
    rotationMode=BMM_Additive
    rotationSpace=BCS_BoneSpace
    translationMode=BMM_Additive
    translationSpace=BCS_BoneSpace
    alphaBoolBlend {blendOption=Linear}
    boneToModify {boneName=dyn_hair_bk_MSTR}
    rotation=<rot 1.1229 39.974 -2.1071>
   }
   n31 = "Transform (Modify) Bone - Bone: dyn_hair_bk_mid_MSTR_R"(ComponentPose=@n30, Alpha=1, Translation="0,0,0", Rotation="15.081,104.195,74.830") {
    rotationMode=BMM_Additive
    rotationSpace=BCS_BoneSpace
    translationMode=BMM_Additive
    translationSpace=BCS_BoneSpace
    alphaBoolBlend {blendOption=Linear}
    boneToModify {boneName=dyn_hair_bk_mid_MSTR_R}
    rotation=<rot 15.081 104.1949 74.8299>
   }
   n32 = "Transform (Modify) Bone - Bone: dyn_hair_MSTR_R"(ComponentPose=@n31, Alpha=1, Translation="0,0,0", Rotation="3.000,90.996,17.196") {
    rotationMode=BMM_Additive
    rotationSpace=BCS_BoneSpace
    translationMode=BMM_Additive
    translationSpace=BCS_BoneSpace
    alphaBoolBlend {blendOption=Linear}
    boneToModify {boneName=dyn_hair_MSTR_R}
    rotation=<rot 2.9999 90.9958 17.1959>
   }
   n33 = "Transform (Modify) Bone - Bone: dyn_hair_MSTR_L"!(ComponentPose=@n32, Alpha=1, Translation="0,0,0", Rotation="0,30,0") -> (Pose) {
    rotationMode=BMM_Additive
    rotationSpace=BCS_BoneSpace
    translationMode=BMM_Additive
    translationSpace=BCS_BoneSpace
    alphaBoolBlend {blendOption=Linear}
    boneToModify {boneName=dyn_hair_MSTR_L}
    rotation {yaw=30}
   }
   Save cached pose 'common cache'(Pose=Component To Local(ComponentPose=@n33)) {cacheName=common cache}
   n34 = "Transform (Modify) Bone - Bone: clavicle_phs_l"!(ComponentPose=Local To Component(LocalPose=Use cached pose 'main cache'()), Alpha=1, Translation="0,0,-3") -> (Pose) {
    translationMode=BMM_Additive
    translationSpace=BCS_BoneSpace
    alphaBoolBlend {blendOption=Linear}
    boneToModify {boneName=clavicle_phs_l}
    rotation {pitch=40}
    translation {z=-3}
   }
   n35 = "Transform (Modify) Bone - Bone: clavicle_phs_r"!(ComponentPose=@n34, Alpha=1, Translation="0,0,3") -> (Pose) {
    translationMode=BMM_Additive
    translationSpace=BCS_BoneSpace
    alphaBoolBlend {blendOption=Linear}
    boneToModify {boneName=clavicle_phs_r}
    rotation {pitch=40}
    translation {z=3}
   }
   n36 = "Transform (Modify) Bone - Bone: clavicle_phs_l"!(ComponentPose=@n35, Alpha=1, Translation="0,0,-3.5", Rotation:Rotator="0,0,0") -> (Pose) {
    rotationMode=BMM_Additive
    rotationSpace=BCS_BoneSpace
    translationMode=BMM_Additive
    translationSpace=BCS_BoneSpace
    alphaBoolBlend {blendOption=Linear}
    boneToModify {boneName=clavicle_phs_l}
    translation {z=-3.5}
   }
   n37 = "Transform (Modify) Bone - Bone: clavicle_phs_r"!(ComponentPose=@n36, Alpha=1, Translation="0,0,3.5", Rotation:Rotator="0,0,0") -> (Pose) {
    rotationMode=BMM_Additive
    rotationSpace=BCS_BoneSpace
    translationMode=BMM_Additive
    translationSpace=BCS_BoneSpace
    alphaBoolBlend {blendOption=Linear}
    boneToModify {boneName=clavicle_phs_r}
    translation {z=3.5}
   }
   n38 = "Transform (Modify) Bone - Bone: dyn_hair_bk_MSTR"(ComponentPose=@n37, Alpha=1, Translation="0,0,0", Rotation="-20,50,0") {
    rotationMode=BMM_Additive
    rotationSpace=BCS_BoneSpace
    translationMode=BMM_Additive
    translationSpace=BCS_BoneSpace
    alphaBoolBlend {blendOption=Linear}
    boneToModify {boneName=dyn_hair_bk_MSTR}
    rotation {pitch=-20, yaw=50}
   }
   n39 = "Transform (Modify) Bone - Bone: dyn_hair_bk_mid_MSTR_R"(ComponentPose=@n38, Alpha=1, Translation="0,0,0", Rotation="7,79.868,73.095") {
    rotationMode=BMM_Additive
    rotationSpace=BCS_BoneSpace
    translationMode=BMM_Additive
    translationSpace=BCS_BoneSpace
    alphaBoolBlend {blendOption=Linear}
    boneToModify {boneName=dyn_hair_bk_mid_MSTR_R}
    rotation=<rot 7 79.868 73.095>
   }
   n40 = "Transform (Modify) Bone - Bone: dyn_hair_MSTR_R"(ComponentPose=@n39, Alpha=1, Translation="0,0,0", Rotation="29.249,45.713,45.582") {
    rotationMode=BMM_Additive
    rotationSpace=BCS_BoneSpace
    translationMode=BMM_Additive
    translationSpace=BCS_BoneSpace
    alphaBoolBlend {blendOption=Linear}
    boneToModify {boneName=dyn_hair_MSTR_R}
    rotation=<rot 29.249 45.7129 45.582>
   }
   n41 = "Transform (Modify) Bone - Bone: dyn_hair_MSTR_L"(ComponentPose=@n40, Alpha=1, Translation="0,0,0", Rotation="0,30,0") {
    rotationMode=BMM_Additive
    rotationSpace=BCS_BoneSpace
    translationMode=BMM_Additive
    translationSpace=BCS_BoneSpace
    alphaBoolBlend {blendOption=Linear}
    boneToModify {boneName=dyn_hair_MSTR_L}
    rotation {yaw=30}
   }
   Save cached pose 'crouching'(Pose=Component To Local(ComponentPose=@n41)) {cacheName=crouching}
  ]}
 EventGraph {nodes [
   n1 = Event BlueprintUpdateAnimation()
   > Set Delta Time X(Delta Time X=@n1.DeltaTimeX)
   > n2 = Sequence!(execute) -> (then_0, then_1, then_2:Exec)
   > n3 = Gravity Override State Function DEPRECATED(
    CharacterPartAnimInstance=Self-Reference()
    GameTimeInSeconds=GetGameTimeInSeconds()
    GravityOverrideParams=Get Gravity Override Params()
    RBAN Params=Get RBAN Params()
   )
   > Set Gravity Override(Gravity Override=@n3)
   > Set RBAN Settings Out(RBAN Settings Out:RBANSetting Struct Structure=@n3.RBAN SettingsOut) -> (Output_Get:RBANSetting Struct Structure)
   n4 = Get Delta Time X()
   n5 = Get Crouching Moving() -> (Crouching Moving:double)
   n6 = MapRangeClamped(
    Value="Get Clavicle R Pitch "() -> ("Clavicle R Pitch ":double)
    InRangeA=1
    InRangeB=0
    OutRangeB=Get Head Back Extra Yaw() -> (Head Back Extra Yaw:double)
   )
   n7 = Get bIsCrouching()
   n8 = Get bIsMoving2D()
   > Set Crouching Moving(Crouching Moving:double=SelectFloat(A=@n6, B=FInterpTo(Current=@n5, Target=0, DeltaTime=@n4, InterpSpeed=2), bPickA=AND Boolean(A=@n7, B=@n8))) -> (Output_Get:double)
   n9 = NOT Boolean(A=Get bIsMoving2D())
   n10 = Get bIsCrouching()
   > Set Crouching Not Moving(Crouching Not Moving:double=SelectFloat(A="Get Head Pitch R "() -> ("Head Pitch R ":double), B=0, bPickA=AND Boolean(A=@n10, B=@n9))) -> (Output_Get:double)
   n11 = GetOwnerSkeletalMeshComponent()
   n12 = BreakTransform!(InTransform=GetDeltaTransformFromRefPose(self=@n11, BoneName=head, BaseName=spine_05)) -> (Location, Rotation_Roll:float, Rotation_Pitch:float, Rotation_Yaw:float, Scale)
   "Set Head Rotation Y (Pitch)"(execute=@n2.then_1, "Head Rotation Y (Pitch)":float=@n12.Rotation_Pitch) -> (Output_Get:float)
   > "Set Head Rotation Z (Yaw)"("Head Rotation Z (Yaw)":float=@n12.Rotation_Yaw) -> (Output_Get:float)
   > "Set Clavicle R Rotation Y (Pitch)"("Clavicle R Rotation Y (Pitch)":float=BreakTransform(InTransform=GetDeltaTransformFromRefPose(self=@n11, BoneName=clavicle_r, BaseName=spine_05)).Rotation_Pitch) -> (Output_Get:float)
   > "Set Clavicle L Rotation Y (Pitch)"("Clavicle L Rotation Y (Pitch)":float=BreakTransform(InTransform=GetDeltaTransformFromRefPose(self=@n11, BoneName=clavicle_l, BaseName=spine_05)).Rotation_Pitch) -> (Output_Get:float)
   Set Head Back Extra Yaw(
    execute=@n2.then_2
    Head Back Extra Yaw:double=MapRangeClamped(Value="Get Head Rotation Z (Yaw)"() -> ("Head Rotation Z (Yaw)":float), InRangeA=-50, InRangeB=-90)
   ) -> (Output_Get:double)
   n13 = MapRangeClamped(Value="Get Head Rotation Z (Yaw)"(), InRangeA=0, InRangeB=10)
   n14 = MapRangeClamped(Value="Get Clavicle R Rotation Y (Pitch)"() -> ("Clavicle R Rotation Y (Pitch)":float), InRangeA=5, InRangeB=15)
   > Set Head Front Yaw(Head Front Yaw:double=MapRangeClamped(Value=float + float(A=@n14, B=@n13), InRangeA=1, InRangeB=2)) -> (Output_Get:double)
   > Set Head Back Yaw(Head Back Yaw:double=MapRangeClamped(Value="Get Head Rotation Z (Yaw)"(), InRangeA=0, InRangeB=-30)) -> (Output_Get:double)
   n15 = MapRangeClamped(Value="Get Clavicle R Rotation Y (Pitch)"(), InRangeA=5, InRangeB=15)
   n16 = MapRangeClamped(Value="Get Head Rotation Y (Pitch)"() -> ("Head Rotation Y (Pitch)":float), InRangeA=20, InRangeB=50)
   > "Set Head Pitch R "("Head Pitch R ":double=MapRangeClamped(Value=float + float(A=@n16, B=@n15), InRangeA=1, InRangeB=2)) -> (Output_Get:double)
   > Set Clavicle L Pitch(Clavicle L Pitch:double=MapRangeClamped(Value="Get Clavicle L Rotation Y (Pitch)"() -> ("Clavicle L Rotation Y (Pitch)":float), InRangeA=5, InRangeB=15)) -> (Output_Get:double)
   > "Set Clavicle R Pitch "("Clavicle R Pitch ":double=MapRangeClamped(Value="Get Clavicle R Rotation Y (Pitch)"(), InRangeA=5, InRangeB=15)) -> (Output_Get:double)
   n17 = MapRangeClamped(Value="Get Clavicle R Rotation Y (Pitch)"(), InRangeA=5, InRangeB=15)
   n18 = MapRangeClamped(Value="Get Head Rotation Z (Yaw)"(), InRangeA=30, InRangeB=-30)
   > Set Clavicle R Pitch Compenstion(Clavicle R Pitch Compenstion:double=MapRangeClamped(Value=float + float(A=@n17, B=@n18), InRangeA=0, InRangeB=2)) -> (Output_Get:double)
   FInterpTo(Current=Get Crouching Moving(), Target=@n6, DeltaTime=Get Delta Time X(), InterpSpeed=10)
  ]}
]
variables [
 BodyMesh {type=SkeletalMeshComponent}
 Clavicle L Pitch {type=number}
 "Clavicle L Rotation Y (Pitch)" {type=number}
 "Clavicle R Pitch " {type=number}
 Clavicle R Pitch Compenstion {type=number}
 "Clavicle R Rotation Y (Pitch)" {type=number}
 Crouching Moving {type=number}
 Crouching Not Moving {type=number}
 Delta Time X {type=number}
 Gravity Override {type=Vector}
 Gravity Override Params {
  type=GravityOverrideParamsStruct
  default {
   skydive_DivingDown {gravityOverride=<-1500 500 0>}
   skydive_FloatingDown {gravityOverride=<0 -350 300>, windAmplitude=<100 -250 200>, windFrequency=<0.75 3.5 4>, windOffset=<0 -300 200>}
   skydive_GoingUp {gravityOverride=<0 -150 750>, windAmplitude=<100 -250 200>, windFrequency=<0.75 3.5 4>}
   skydive_Parachuting {gravityOverride=<0 -400 -980>, windAmplitude=<100 -250 200>, windFrequency=<0.75 3.5 4>}
  }
 }
 Head Back Extra Yaw {type=number}
 Head Back Yaw {type=number}
 Head Front Yaw {type=number}
 "Head Pitch R " {type=number}
 "Head Rotation Y (Pitch)" {type=number}
 "Head Rotation Z (Yaw)" {type=number}
 RBAN Params {
  type=RBANParamsStruct
  default {
   dBNO {
    componentLinearAccScale=<0.5 0.25 0.5>
    componentLinearVelScale=<0.5 0.25 0.5>
    simSpaceSettings {
     maxAngularAcceleration=1000
     maxAngularVelocity=3
     maxLinearAcceleration=300
     maxLinearVelocity=700
     velocityScaleZ=0
     externalLinearDragV=<0 0.5 0>
    }
   }
   dBNO_Carried {
    componentLinearAccScale=<0.5 0.25 0.5>
    componentLinearVelScale=<0.5 0.25 0.5>
    simSpaceSettings {
     maxAngularAcceleration=1000
     maxAngularVelocity=3
     maxLinearAcceleration=300
     maxLinearVelocity=700
     velocityScaleZ=0
     externalLinearDragV=<0 0.5 0>
    }
   }
   emoteOrMelee {
    componentLinearAccScale=<0.5 0.25 0.5>
    componentLinearVelScale=<0.5 0.25 0.5>
    simSpaceSettings {
     maxAngularAcceleration=1000
     maxAngularVelocity=3
     maxLinearAcceleration=300
     maxLinearVelocity=700
     velocityScaleZ=0
     externalLinearDragV=<0 0.5 0>
    }
   }
   front_End {
    componentLinearAccScale=<0.5 0.25 0.5>
    componentLinearVelScale=<0.5 0.25 0.5>
    simSpaceSettings {
     maxAngularAcceleration=1000
     maxAngularVelocity=3
     maxLinearAcceleration=300
     maxLinearVelocity=700
     velocityScaleZ=0
     externalLinearDragV=<0 0.5 0>
    }
   }
   onGround_Crouch_Moving {
    componentLinearAccScale=<0.5 0.25 0.5>
    componentLinearVelScale=<0.5 0.25 0.5>
    simSpaceSettings {
     maxAngularAcceleration=1000
     maxAngularVelocity=3
     maxLinearAcceleration=300
     maxLinearVelocity=700
     velocityScaleZ=0
     externalLinearDragV=<0 0.5 0>
    }
   }
   onGround_Falling {
    componentLinearAccScale=<0.5 0.25 0.5>
    componentLinearVelScale=<0.5 0.25 0.5>
    simSpaceSettings {
     maxAngularAcceleration=1000
     maxAngularVelocity=3
     maxLinearAcceleration=300
     maxLinearVelocity=700
     velocityScaleZ=0
     externalLinearDragV=<0 0.5 0>
    }
   }
   onGround_Moving {
    componentLinearAccScale=<0.5 0.15 0.5>
    componentLinearVelScale=<0.5 0.15 0.5>
    simSpaceSettings {
     maxAngularAcceleration=1000
     maxAngularVelocity=3
     maxLinearAcceleration=300
     maxLinearVelocity=700
     velocityScaleZ=0
     externalLinearDragV=<0 0.5 0>
    }
   }
   onGround_Standing {
    componentLinearAccScale=<0.5 0.15 0.5>
    componentLinearVelScale=<0.5 0.15 0.5>
    simSpaceSettings {
     maxAngularAcceleration=1000
     maxAngularVelocity=3
     maxLinearAcceleration=300
     maxLinearVelocity=700
     velocityScaleZ=0
     externalLinearDragV=<0 0.5 0>
    }
   }
   rail_Grinding {
    componentLinearAccScale=<0.5 0.25 0.5>
    componentLinearVelScale=<0.5 0.25 0.5>
    simSpaceSettings {
     maxAngularAcceleration=1000
     maxAngularVelocity=3
     maxLinearAcceleration=300
     maxLinearVelocity=700
     velocityScaleZ=0
     worldAlpha=1
     externalLinearDragV=<0 0.5 0>
    }
   }
   riding_Moving {
    componentLinearAccScale=<0.5 0.25 0.5>
    componentLinearVelScale=<0.5 0.25 0.5>
    simSpaceSettings {
     maxAngularAcceleration=1000
     maxAngularVelocity=3
     maxLinearAcceleration=300
     maxLinearVelocity=700
     velocityScaleZ=0
     externalLinearDragV=<0 0.5 0>
    }
   }
   riding_Moving_Targeting {
    componentLinearAccScale=<0.5 0.25 0.5>
    componentLinearVelScale=<0.5 0.25 0.5>
    simSpaceSettings {
     maxAngularAcceleration=1000
     maxAngularVelocity=3
     maxLinearAcceleration=300
     maxLinearVelocity=700
     velocityScaleZ=0
     externalLinearDragV=<0 0.5 0>
    }
   }
   riding_Standing {
    componentLinearAccScale=<0.5 0.25 0.5>
    componentLinearVelScale=<0.5 0.25 0.5>
    simSpaceSettings {
     maxAngularAcceleration=1000
     maxAngularVelocity=3
     maxLinearAcceleration=300
     maxLinearVelocity=700
     velocityScaleZ=0
     externalLinearDragV=<0 0.5 0>
    }
   }
   skydive_DivingDown {
    componentLinearAccScale=<0.25 0.25 0.25>
    componentLinearVelScale=<0.25 0.25 0.25>
    simSpaceSettings {
     maxAngularAcceleration=1000
     maxAngularVelocity=3
     maxLinearAcceleration=300
     maxLinearVelocity=700
     velocityScaleZ=0
     externalLinearDragV=<0 0.5 0>
    }
   }
   skydive_FloatingDown {
    componentLinearAccScale=<0.25 0.25 0.25>
    componentLinearVelScale=<0.25 0.25 0.25>
    simSpaceSettings {
     maxAngularAcceleration=1000
     maxAngularVelocity=3
     maxLinearAcceleration=300
     maxLinearVelocity=700
     velocityScaleZ=0
     externalLinearDragV=<0 0.5 0>
    }
   }
   skydive_GoingUp {
    componentLinearAccScale=<0.25 0.25 0.25>
    componentLinearVelScale=<0.25 0.25 0.25>
    simSpaceSettings {
     maxAngularAcceleration=1000
     maxAngularVelocity=3
     maxLinearAcceleration=300
     maxLinearVelocity=700
     velocityScaleZ=0
     externalLinearDragV=<0 0.5 0>
    }
   }
   skydive_Parachuting {
    componentLinearAccScale=<0.25 0.25 0.25>
    componentLinearVelScale=<0.25 0.25 0.25>
    simSpaceSettings {
     maxAngularAcceleration=1000
     maxAngularVelocity=3
     maxLinearAcceleration=300
     maxLinearVelocity=700
     velocityScaleZ=0
     externalLinearDragV=<0 0.5 0>
    }
   }
   skydive_Surfing {
    componentLinearAccScale=<0.5 0.25 0.5>
    componentLinearVelScale=<0.5 0.25 0.5>
    simSpaceSettings {
     maxAngularAcceleration=1000
     maxAngularVelocity=3
     maxLinearAcceleration=300
     maxLinearVelocity=700
     velocityScaleZ=0
     externalLinearDragV=<0 0.5 0>
    }
   }
   swimming {
    componentLinearAccScale=<0.5 0.25 0.5>
    componentLinearVelScale=<0.5 0.25 0.5>
    simSpaceSettings {
     maxAngularAcceleration=1000
     maxAngularVelocity=3
     maxLinearAcceleration=300
     maxLinearVelocity=700
     velocityScaleZ=0
     externalLinearDragV=<0 0.5 0>
    }
   }
   tactical_Sprint {
    componentLinearAccScale=<0.5 0.25 0.5>
    componentLinearVelScale=<0.5 0.25 0.5>
    simSpaceSettings {
     maxAngularAcceleration=1000
     maxAngularVelocity=3
     maxLinearAcceleration=300
     maxLinearVelocity=700
     velocityScaleZ=0
     externalLinearDragV=<0 0.5 0>
    }
   }
   ziplining {
    componentLinearAccScale=<0.5 0.25 0.5>
    componentLinearVelScale=<0.5 0.25 0.5>
    simSpaceSettings {
     maxAngularAcceleration=1000
     maxAngularVelocity=3
     maxLinearAcceleration=300
     maxLinearVelocity=700
     velocityScaleZ=0
     externalLinearDragV=<0 0.5 0>
    }
   }
  }
 }
 RBAN Settings Out {
  type=RBANSettingStruct
  default {
   componentLinearAccScale=<0 0 0>
   componentLinearVelScale=<0 0 0>
   simSpaceSettings {
    dampingAlpha=1
    maxAngularAcceleration=1000
    maxAngularVelocity=3
    maxLinearAcceleration=300
    maxLinearVelocity=700
    velocityScaleZ=0
    worldAlpha=1
    externalAngularVelocity=<0 0 0>
    externalLinearDragV=<0 0.5 0>
    externalLinearVelocity=<0 0 0>
   }
  }
 }
]
